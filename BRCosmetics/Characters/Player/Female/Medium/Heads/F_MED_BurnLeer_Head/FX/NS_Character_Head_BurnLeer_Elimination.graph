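# Niagara elimination effect for the female BurnLeer character head, combining an ember burst, looping rising embers, two mesh fire layers and glowing eye sprites.
# Emitters follow head mesh sockets through a shared skeletal mesh reader, with offsets and scales adjusted by BigHeadMode.
# Eye glow alpha is gated per eye by a closing-eyes flag, with aiming down sights suppressing one eye.
# A head-hidden flag zeroes every emitter's alpha.
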NiagaraSystem /BRCosmetics/Characters/Player/Female/Medium/Heads/F_MED_BurnLeer_Head/FX/NS_Character_Head_BurnLeer_Elimination
dependencies {
 usedDataInterfaces [
  &NiagaraDataInterfaceCurve
  &NiagaraDataInterfaceSpriteRendererInfo
  &NiagaraDataInterfaceColorCurve
  &NiagaraDataInterfaceVectorCurve
  &NiagaraDataInterfaceArrayInt32
  &NiagaraDataInterfaceArrayFloat3
  &NiagaraDataInterfaceSkeletalMesh
 ]
 usedDynamicInputs [
  &SetBoolByFloatComparison
  &Boolean_AndOperation
  &SetBoolByIntComparison
 ]
 usedModules [
  &self:SystemSpawnScript.SystemScriptSource.SystemScriptGraph.NiagaraNodeAssignment_0.SetVariables_#1
  &self:SystemSpawnScript.SystemScriptSource.SystemScriptGraph.NiagaraNodeAssignment_1.SetVariables_#2
  &SystemState
  /Game/Effects/Niagara/Modules/Logic/UpdateValueStateChange
  &self:DistanceRangeFromBoneToBone
  &self:SystemSpawnScript.SystemScriptSource.SystemScriptGraph.NiagaraNodeAssignment_2.SetVariables_#3
  &Timeline
 ]
}
emitters {
 NE_Burst_Embers {
  emitterUpdateScript [
   &EmitterState()
   &SpawnBurst_Instantaneous(Spawn Count=15, Use Spawn Probability=false, UseLoopCountLimit=false)
  ]
  particleSpawnScript [
   &InitializeParticle(
    Color Mode=::NewEnumerator0
    Lifetime Max=0.9
    Lifetime Min=0.5
    Lifetime Mode=::NewEnumerator1
    Sprite Rotation Mode=::NewEnumerator2
    Sprite Size Mode=::NewEnumerator1
    Uniform Sprite Size Max=15
    UsePositionOffset=false
    Write Material Random=true
   )
   &AddVelocity(
    Module.Use Falloff Distance=false
    Override Seed=false
    Velocity Mode=::NewEnumerator1
    Velocity Origin=$Particles.SkeletalMeshLocation.SampledPosition
    Velocity Speed=&Lerp_Float(
     A=&RandomRangeFloat(Minimum=25, Maximum=150)
     B=&RandomRangeFloat(Minimum=50, Maximum=300)
     Alpha=$NPC.FortniteNPC.BigHeadMode
    )
   )
  ]
  particleUpdateScript [
   &ParticleState()
   &SkeletalMeshLocation(
    Bone / Socket Index=2
    MaskSpawnGroup=false
    Mesh Sampling Type=::NewEnumerator5
    Position Sampling=::NewEnumerator1
    Sampled Position Offset=&Lerp_Vector(A=<5 0 0>, B=<10 0 0>, Alpha=$NPC.FortniteNPC.BigHeadMode)
    Skeletal Mesh=$System.Skeletal Mesh Reader
    Socket Sampling Mode=::NewEnumerator5
   )
   &ScaleSpriteSize(
    Uniform Curve Scale=&Lerp_Float(A=1, B=2, Alpha=$NPC.FortniteNPC.BigHeadMode)
    Uniform Curve Sprite Scale=DataInterface() {propertyValues {Curve {
       defaultValue=0
       keys [
        {arriveTangent=9.0725, interpMode=RCIM_Cubic, leaveTangent=9.0725, tangentMode=RCTM_User, time=0, value=0}
        {arriveTangent=-2.5249, interpMode=RCIM_Cubic, leaveTangent=-2.5249, tangentMode=RCTM_User, time=0.1408, value=0.9847}
        {arriveTangent=-0.0405, interpMode=RCIM_Cubic, leaveTangent=-0.0405, tangentMode=RCTM_User, time=1, value=0}
       ]
      }}}
   )
   &Drag(Drag=10, Use Linear Drag=true, Use Rotational Drag=true)
   &GravityForce(Gravity=&RandomRangeVector(Minimum=<0 0 250>, Maximum=<0 1 500>))
   &CurlNoiseForce(
    Bias Noise=false
    Noise Frequency=20
    Noise Strength=&RandomRangeFloat(Minimum=50, Maximum=150)
    Pan Noise=false
   )
   &SolveForcesAndVelocity(Clamp Velocity=false, Limit Acceleration=false)
   &DynamicMaterialParameters(
    Index 0 Param 1=&FloatFromCurve(FloatCurve=DataInterface() {propertyValues {Curve {
        defaultValue=1
        keys [{interpMode=RCIM_Cubic, tangentMode=RCTM_User, time=0, value=0}, {interpMode=RCIM_Cubic, tangentMode=RCTM_User, time=1.0018, value=0.5073}]
       }}})
    Index 0 Param 2=&RandomRangeFloat(Minimum=0, Maximum=1)
    Index 0 Param 3=&RandomRangeFloat(Minimum=0, Maximum=1)
    Module.Index0Float0Write=true
    Module.Index0Float1Write=true
    Module.Index0Float2Write=true
    Module.Index0Float3Write=true
   )
   &ScaleColor(
    Scale Alpha=&FloatFromCurve(
     FloatCurve=DataInterface() {propertyValues {Curve {keys [<at 0 0>,<at 0.5 1>,<at 1 0>]}}}
     Scale Curve=&Lerp_Float(A=1, B=0, Alpha=&MakeCustomFloatFromBool(Bool=$User.bisHeadHidden, True Float=1, False Float=0))
    )
    ScaleA=true
    ScaleRGB=true
   )
   &CameraOffset(Camera Offset Amount=5)
  ]
  rendererClasses [&NiagaraSpriteRendererProperties]
  renderers [{rendererClass=&NiagaraSpriteRendererProperties}]
 }
 NE_Embers {
  emitterUpdateScript [
   &EmitterState(
    Inactive Response=::NewEnumerator1
    Life Cycle Mode=::NewEnumerator1
    Loop Behavior=::NewEnumerator1
    Loop Delay=0.05
    Loop Duration=1.5
    UseLoopDelay=true
   )
   &SpawnRate(SpawnRate=10, Use Spawn Probability=false)
  ]
  particleSpawnScript [
   &InitializeParticle(
    Color=<rgba 5 0.25 0 1>
    Lifetime Max=1.5
    Lifetime Mode=::NewEnumerator1
    Position Offset=<0 0 30>
    Position Offset Coordinate Space=::Local
    Sprite Rotation Mode=::NewEnumerator2
    Sprite Size Mode=::NewEnumerator1
    Uniform Sprite Size Max=3.25
    Uniform Sprite Size Min=1.75
    UsePositionOffset=true
    Write Material Random=true
   )
   &ShapeLocation(MaskSpawnGroup=false, Override Seed=false, Sphere Radius=10, UseOwnerScale=true)
   &SkeletalMeshLocation(
    Bone / Socket Index=2
    MaskSpawnGroup=false
    Mesh Sampling Type=::NewEnumerator5
    Skeletal Mesh=$System.Skeletal Mesh Reader
    Socket Sampling Mode=::NewEnumerator5
   )
  ]
  particleUpdateScript [
   &ParticleState()
   &ScaleSpriteSize(
    Uniform Curve Scale=&Lerp_Float(A=1, B=1.5, Alpha=$NPC.FortniteNPC.BigHeadMode)
    Uniform Curve Sprite Scale=DataInterface() {propertyValues {Curve {keys [
        {arriveTangent=0.5516, interpMode=RCIM_Cubic, leaveTangent=0.5516, tangentMode=RCTM_User, time=0, value=0}
        {arriveTangent=5.476, interpMode=RCIM_Cubic, leaveTangent=-0.224, tangentMode=RCTM_Break, time=0.4001, value=1.0058}
        {arriveTangent=0.0706, interpMode=RCIM_Cubic, leaveTangent=-23.6527, tangentMode=RCTM_Break, time=0.9106, value=1}
        <at 1 0>
       ]}}}
   )
   &SpriteRotationRate(Rotation Rate=&AngleConversion(
     Angle:NiagaraFloat=&RandomRangeFloat(Minimum=-1, Maximum=1)
     Angle Input:ENiagara_AngleInput=::NewEnumerator1
     Angle Output:ENiagara_AngleInput=::NewEnumerator0
    ))
   &Drag(Drag=&RandomRangeFloat(Minimum=6, Maximum=8), Use Linear Drag=true, Use Rotational Drag=true)
   &AccelerationForce(Acceleration=&Lerp_Vector(A=<0 0 80>, B=<0 0 240>, Alpha=$NPC.FortniteNPC.BigHeadMode))
   &SolveForcesAndVelocity(Clamp Velocity=false, Limit Acceleration=false)
   &SubUVAnimation(
    Random Start Frame=true
    "Sprite Renderer "=DataInterface() {propertyValues {SpriteRenderer=&self:NE_Embers_0.NiagaraSpriteRendererProperties_0}}
    UseEndFrame=false
    UseStartFrame=false
   )
   &ScaleColor(
    Scale Alpha=&FloatFromCurve(
     FloatCurve=DataInterface() {propertyValues {Curve {keys [<at 0 0>,<at 0.2 1>,<at 1 0.2>]}}}
     Scale Curve=&Lerp_Float(A=1, B=0, Alpha=&MakeCustomFloatFromBool(Bool=$User.bisHeadHidden, True Float=1, False Float=0))
    )
    Scale RGB=&VectorFromCurve(VectorCurve=DataInterface() {propertyValues {
       XCurve {keys [
         <at 0 500 RCIM_Cubic>
         {arriveTangent=-500, interpMode=RCIM_Cubic, leaveTangent=11.6698, tangentMode=RCTM_Break, time=0.8, value=0}
         <at 1 0 RCIM_Cubic>
        ]}
       YCurve {keys [
         <at 0 62.9696 RCIM_Cubic>
         {arriveTangent=-62.9696, interpMode=RCIM_Cubic, leaveTangent=-0.0144, tangentMode=RCTM_Break, time=0.8, value=0}
         <at 1 0 RCIM_Cubic>
        ]}
       ZCurve {keys [<at 0 0 RCIM_Cubic>,<at 1 0 RCIM_Cubic>]}
      }})
    ScaleA=true
    ScaleRGB=true
   )
  ]
  rendererClasses [&NiagaraSpriteRendererProperties]
  renderers [{rendererClass=&NiagaraSpriteRendererProperties}]
 }
 NE_EyesGlow {
  emitterUpdateScript [
   &EmitterState()
   &SpawnBurst_Instantaneous(Spawn Count=2, Use Spawn Probability=false, UseLoopCountLimit=false)
  ]
  particleSpawnScript [
   &PartitionParticles()
   &InitializeParticle(
    Color=<rgba 15 2.3478 0 1>
    Lifetime=6
    Sprite Size=<8 5>
    Sprite Size Mode=::NewEnumerator4
    UsePositionOffset=false
    Write Material Random=true
   )
  ]
  particleUpdateScript [
   &ParticleState()
   &ScaleSpriteSize(
    Scale Sprite Size Mode=::NewEnumerator0
    Uniform Scale Factor=&Lerp_Float(A=1, B=2, Alpha=$NPC.FortniteNPC.BigHeadMode)
   )
   &SkeletalMeshLocation(
    Bone / Socket Index=&SelectIntFromArray(Direct Array Index=$Particles.PartitionParticles.Partition, Int Selection Array=DataInterface() {propertyValues {IntData [0,1]}})
    MaskSpawnGroup=false
    Mesh Sampling Type=::NewEnumerator5
    Sampled Position Offset=&Lerp_Vector(
     A=&SelectVectorFromArray(
      Direct Array Index=$Particles.PartitionParticles.Partition
      Vector Selection Array=DataInterface() {propertyValues {FloatData [<-2.25 -1 0.5>, <-2.25 -1 -0.5>]}}
     )
     B=&SelectVectorFromArray(
      Direct Array Index=$Particles.PartitionParticles.Partition
      Vector Selection Array=DataInterface() {propertyValues {FloatData [<-4.5 -1 0.5>, <-4.5 -1 -0.5>]}}
     )
     Alpha=$NPC.FortniteNPC.BigHeadMode
    )
    Skeletal Mesh=$System.Skeletal Mesh Reader
    Socket Sampling Mode=::NewEnumerator5
   )
   &ScaleColor(
    Scale Alpha=&FloatFromCurve(
     FloatCurve=DataInterface() {propertyValues {Curve {keys [<at 0 0>,<at 0.4 1>,<at 0.5 1>,<at 1 0>]}}}
     Scale Curve=&Lerp_Float(
      A=&MergeHelper_FloatSwitch(
       0=&MakeCustomFloatFromBool(Bool=$System.bIsClosingEyes, True Float=1, False Float=0)
       1=&MakeCustomFloatFromBool(
        Bool=$User.bIsAds
        True Float=0
        False Float=&MakeCustomFloatFromBool(Bool=$System.bIsClosingEyes, True Float=1, False Float=0)
       )
       2=?
       3=?
       Index=$Particles.PartitionParticles.Partition
       Num=::NewEnumerator0
      )
      B=0
      Alpha=&MakeCustomFloatFromBool(Bool=$User.bisHeadHidden, True Float=1, False Float=0)
     )
    )
    Scale RGB=&VectorFromFloat(Value=&FloatFromCurve(FloatCurve=DataInterface() {propertyValues {Curve {keys [<at 0 0>,<at 0.4 1>,<at 0.5 1>,<at 1 0>]}}}))
    ScaleA=true
    ScaleRGB=true
   )
   &CameraOffset(Camera Offset Amount=2)
  ]
  rendererClasses [&NiagaraSpriteRendererProperties]
  renderers [
   {rendererClass=&NiagaraSpriteRendererProperties}
   {rendererClass=&NiagaraSpriteRendererProperties, rendererIndex=1}
  ]
 }
 NE_Fire_01 {
  emitterUpdateScript [
   &EmitterState(
    Inactive Response=::NewEnumerator1
    Life Cycle Mode=::NewEnumerator1
    Loop Behavior=::NewEnumerator1
    Loop Delay=0.05
    Loop Duration=6
    UseLoopDelay=true
   )
   &SpawnBurst_Instantaneous(Use Spawn Probability=false, UseLoopCountLimit=false)
  ]
  particleSpawnScript [
   &InitializeParticle(
    Color=<rgba 1 0.6158 0.257 2>
    Lifetime=6
    Mesh Scale=<0.325 0.8 0.25>
    Mesh Scale Mode=::NewEnumerator4
    UsePositionOffset=false
    Write Material Random=true
   )
   &InitialMeshOrientation(Mesh Axis To Orient=<0 1 0>, Mesh Orientation Mode=::NewEnumerator2, Orientation Vector=<0 1 0>, Use Rotation Vector=true)
   &CameraOffset(Camera Offset Amount=-20)
  ]
  particleUpdateScript [
   &ParticleState()
   &SkeletalMeshLocation(
    Bone / Socket Index=2
    MaskSpawnGroup=false
    Mesh Sampling Type=::NewEnumerator5
    Sampled Position Offset=<-5 0 0>
    Skeletal Mesh=$System.Skeletal Mesh Reader
    Socket Sampling Mode=::NewEnumerator5
   )
   &ScaleMeshSize(Scale Factor=&VectorFromFloat(Value=&Lerp_Float(A=1, B=2, Alpha=$NPC.FortniteNPC.BigHeadMode)))
   &ScaleColor(
    Scale Alpha=&FloatFromCurve(
     FloatCurve=DataInterface() {propertyValues {Curve {keys [<at 0 0>,<at 0.4 1>,<at 0.5 1>,<at 1 0>]}}}
     Scale Curve=&Lerp_Float(A=0.5, B=0, Alpha=&MakeCustomFloatFromBool(Bool=$User.bisHeadHidden, True Float=1, False Float=0))
    )
    ScaleA=true
    ScaleRGB=true
   )
   &DynamicMaterialParameters(
    Index 0 Param 1=1
    Module.Index0Float0Write=true
    Module.Index0Float1Write=true
    Module.Index0Float2Write=true
    Module.Index0Float3Write=true
   )
  ]
  rendererClasses [&NiagaraMeshRendererProperties]
  renderers [
   {rendererClass=&NiagaraMeshRendererProperties, rendererIndex=0}
   {rendererClass=&NiagaraMeshRendererProperties, rendererIndex=1}
  ]
 }
 NE_Fire_02 {
  emitterUpdateScript [
   &EmitterState(
    Inactive Response=::NewEnumerator1
    Life Cycle Mode=::NewEnumerator1
    Loop Behavior=::NewEnumerator1
    Loop Delay=0.05
    Loop Duration=6
    UseLoopDelay=true
   )
   &SpawnBurst_Instantaneous(Use Spawn Probability=false, UseLoopCountLimit=false)
   &SpawnBurst_Instantaneous(Spawn Group=1, Use Spawn Probability=false, UseLoopCountLimit=false) {module=SpawnBurst_Instantaneous001}
   &SpawnBurst_Instantaneous(Spawn Group=2, Use Spawn Probability=false, UseLoopCountLimit=false) {module=SpawnBurst_Instantaneous002}
  ]
  particleSpawnScript [
   /SetVariables(Particles.VisibilityTag=&SpawnGroup()) {isSetParameters=true}
   &InitializeParticle(
    Color=<rgba 1 0.6158 0.257 2>
    Lifetime=6
    Mesh Scale=&MergeHelper_VectorSwitch(0=<0.2 0.1 0.2>, 1=<0.15 0.1 0.2>, 2=<0.15 0.1 0.2>, 3=?, Index=$Particles.VisibilityTag, Num=::NewEnumerator1)
    Mesh Scale Mode=::NewEnumerator4
    UsePositionOffset=false
    Write Material Random=true
   )
   &InitialMeshOrientation(Mesh Axis To Orient=<0 1 0>, Mesh Orientation Mode=::NewEnumerator2, Orientation Vector=<0 1 0>, Use Rotation Vector=true)
   &CameraOffset(Camera Offset Amount=&MergeHelper_FloatSwitch(0=5, 1=5, 2=5, 3=?, Index=$Particles.VisibilityTag, Num=::NewEnumerator1))
  ]
  particleUpdateScript [
   &ParticleState()
   &SkeletalMeshLocation(
    Bone / Socket Index=2
    MaskSpawnGroup=false
    Mesh Sampling Type=::NewEnumerator5
    Sampled Position Offset=&MergeHelper_VectorSwitch(0=<-5 -5 0>, 1=<2 0 -8>, 2=<2 0 8>, 3=?, Index=$Particles.VisibilityTag, Num=::NewEnumerator1)
    Skeletal Mesh=$System.Skeletal Mesh Reader
    Socket Sampling Mode=::NewEnumerator5
   )
   &ScaleMeshSize(Scale Factor=&VectorFromFloat(Value=&Lerp_Float(A=1, B=2, Alpha=$NPC.FortniteNPC.BigHeadMode)))
   &ScaleColor(
    Scale Alpha=&FloatFromCurve(
     FloatCurve=DataInterface() {propertyValues {Curve {keys [<at 0 0>,<at 0.4 1>,<at 0.5 1>,<at 1 0>]}}}
     Scale Curve=&Lerp_Float(A=0.5, B=0, Alpha=&MakeCustomFloatFromBool(Bool=$User.bisHeadHidden, True Float=1, False Float=0))
    )
    ScaleA=true
    ScaleRGB=true
   )
   &DynamicMaterialParameters(
    Index 0 Param 1=1
    Module.Index0Float0Write=true
    Module.Index0Float1Write=true
    Module.Index0Float2Write=true
    Module.Index0Float3Write=true
   )
  ]
  rendererClasses [&NiagaraMeshRendererProperties]
  renderers [
   {rendererClass=&NiagaraMeshRendererProperties, rendererIndex=0}
   {rendererClass=&NiagaraMeshRendererProperties, rendererIndex=1}
   {rendererClass=&NiagaraMeshRendererProperties, rendererIndex=2}
  ]
 }
}
summary {
 systemName=NS_Character_Head_BurnLeer_Elimination
 emitters [
  {emitterName=NE_Embers, rendererClasses [&NiagaraSpriteRendererProperties]}
  {emitterName=NE_Fire_01, rendererClasses [&NiagaraMeshRendererProperties]}
  {emitterName=NE_Fire_02, rendererClasses [&NiagaraMeshRendererProperties]}
  {emitterName=NE_EyesGlow, rendererClasses [&NiagaraSpriteRendererProperties]}
  {emitterName=NE_Burst_Embers, rendererClasses [&NiagaraSpriteRendererProperties]}
 ]
 userVariables [
  User.bIsAds {type=NiagaraBool}
  User.bIsFrontEnd {type=NiagaraBool}
  User.bIsFrontEndPreview {type=NiagaraBool}
  User.bisHeadHidden {type=NiagaraBool}
  User.IsReactive {type=NiagaraFloat, defaultValue=1, description=""}
  User.IsSkydiving {type=NiagaraBool}
  User.KillCount {type=NiagaraInt32}
 ]
}
